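; Manager actor for the PistolCanary live event zone, driving its pre-event, event and post-event phases.
; If the calendar event is active, it loads the zone data layer, clears nearby chests, AI and pickups, and enables sprite providers.
; It spawns respawn and end-game mutators, with a timer switching to global respawn near event time.
; Calendar end starts event cleanup and rules; a router event ends them. PC_* events duplicate phases for debugging.
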(event Custom|PC_ZoneDL
  ;[0 0.15 0.05105 0.5] If the pre-event calendar event is active, load the event zone data layer.
  (Variables|Cheat|SetLoadedWithCheats true)
  (Development|ExecuteConsoleCommand "Cheat ke* DebugCountdown 1800") ; Seconds in a half hour for timer.
  (bind _outactors (Fortnite|GetAllActorsOfClassinRange "/Game/Building/ActorBlueprints/Containers/Tiered_Chest_Athena_Parent.Tiered_Chest_Athena_Parent_C" (Transformation|GetActorLocation) 1.874e+04))
  (for _array_element _outactors
    (Actor|DestroyActor _array_element))
  (bind _returnvalue (WorldPartition|GetDataLayerManager))
  (DataLayers|SetDataLayerRuntimeState _returnvalue "/PistolCanary_Training/DataLayer/DL_PistolCanary_EventZone.DL_PistolCanary_EventZone" "Activated" false :!class /Script/Engine.DataLayerManager)
  (CallFunction|ActivateMutators)
  (Utilities|FlowControl|Delay 2.0)
  (if (Networking|IsDedicatedServer)
    (Utilities|FlowControl|Sequence
      (:then_0
        ; MMHenchmenManager Disable Henchmen Patrol Path Event
        (bind _returnvalue_2 (GameplayEventRouter|BroadcastEvent (GameplayEventRouter|GetGlobalEventRouter) (Utilities|Struct|MakeMMHenchmenManagerEvents "NewEnumerator1") 0))
        (Fortnite|FortShippingLog self (Utilities|String|Append "EVENT ZONE MANAGER:: MMHenchman DISABLE PATHS EVENT:: ANY LISTENERS? " (Utilities|String|ToString(Boolean) _returnvalue_2))))
      (:then_1
        ; Despawn In-Area AIs
        (bind _outactors_2 (Fortnite|GetAllActorsOfClassinRange "/Script/FortniteAI.FortAthenaPatrolPath" (Transformation|GetActorLocation) 1.874e+04))
        (if (Utilities|Array|IsNotEmpty _outactors_2)
          (for _array_element_2 _outactors_2
            (if (Actor|ActorHasTag _array_element_2 "Henchmen")
              (Fortnite|FortShippingLog self (Utilities|String|Append "Disabled Henchmen: " (Utilities|GetDisplayName _array_element_2)))
              (Class|FortAthenaPatrolPath|SetPatrolPathEnabled _array_element_2 false)
              (Class|FortAthenaLivingWorldPointProviderInterface|DisablePointProvider(Message) _array_element_2)))
          (else
            _))
        (bind _outfortpawns (Fortnite|GetFortPawnsInRange (Transformation|GetActorLocation) 1.874e+04))
        (if (Utilities|Array|IsNotEmpty _outfortpawns)
          (for _array_element_3 _outfortpawns
            (Utilities|Casting|CastToBP_PlayerPawn_NonParticipant _array_element_3
              (:then
                (Actor|DestroyActor _array_element_3))
              (:CastFailed
                (Utilities|Casting|CastToNPC_Pawn_Irwin_Simple_Sprite _array_element_3)
                (Actor|DestroyActor _array_element_3))))
          (else
            _))
        ; To enable Superman Sprite providers
        (bind _outactors_3 (Fortnite|GetAllActorsOfClassinRange "/Script/LagerRuntime.FortAthenaLivingWorldStaticPointProvider" (Transformation|GetActorLocation) 1.874e+04))
        (if (Utilities|Array|IsNotEmpty _outactors_3)
          (for _array_element_4 _outactors_3
            (if (== (Class|FortAthenaLivingWorldPointProviderInterface|GetFiltersTags _array_element_4) (Variables|Default|GetSpriteFilterTag))
              (Class|FortAthenaLivingWorldPointProviderInterface|EnablePointProvider_DEPRECATED _array_element_4)
              (Utilities|Array|AddUnique (Variables|Default|GetSpritePointProviders) _array_element_4))))
        (Utilities|FlowControl|Delay 1.0)
        ; Destroy pickups in the area
        (bind _outactors_4 (Fortnite|GetAllActorsOfClassinRange "/Script/FortniteGame.FortPickup" (Transformation|GetActorLocation) 1.874e+04))
        (if (Utilities|Array|IsNotEmpty _outactors_4)
          (for _array_element_5 _outactors_4
            (Actor|DestroyActor _array_element_5)))))))

(event EventBeginPlay
  ;[0 0.15 0.05105 0.5] If the pre-event calendar event is active, load the event zone data layer.
  (Networking|SwitchHasAuthority
    (:Authority
      (if (Variables|Default|GetFollowCalendarEvents)
        (Class|FortAsyncActionAuthorityMatchReady|MatchReadyAsync
          (:OnReady
            (MCPContext|InitialCalendarSyncCompleted
              (:OnSyncComplete
                (bind _biseventactive (CalendarEvents|GetCalendarEventInformation (Variables|Default|GetCalendarEventToWatch)))
                (if _biseventactive
                  (bind _outactors (Fortnite|GetAllActorsOfClassinRange "/Game/Building/ActorBlueprints/Containers/Tiered_Chest_Athena_Parent.Tiered_Chest_Athena_Parent_C" (Transformation|GetActorLocation) 1.874e+04))
                  (for _array_element _outactors
                    (Actor|DestroyActor _array_element))
                  (bind _returnvalue (WorldPartition|GetDataLayerManager))
                  (DataLayers|SetDataLayerRuntimeState _returnvalue "/PistolCanary_Training/DataLayer/DL_PistolCanary_EventZone.DL_PistolCanary_EventZone" "Activated" false :!class /Script/Engine.DataLayerManager)
                  (CallFunction|ActivateMutators)
                  (Utilities|FlowControl|Delay 2.0)
                  (if (Networking|IsDedicatedServer)
                    (Utilities|FlowControl|Sequence
                      (:then_0
                        ; MMHenchmenManager Disable Henchmen Patrol Path Event
                        (bind _returnvalue_2 (GameplayEventRouter|BroadcastEvent (GameplayEventRouter|GetGlobalEventRouter) (Utilities|Struct|MakeMMHenchmenManagerEvents "NewEnumerator1") 0))
                        (Fortnite|FortShippingLog self (Utilities|String|Append "EVENT ZONE MANAGER:: MMHenchman DISABLE PATHS EVENT:: ANY LISTENERS? " (Utilities|String|ToString(Boolean) _returnvalue_2))))
                      (:then_1
                        ; Despawn In-Area AIs
                        (bind _outactors_2 (Fortnite|GetAllActorsOfClassinRange "/Script/FortniteAI.FortAthenaPatrolPath" (Transformation|GetActorLocation) 1.874e+04))
                        (if (Utilities|Array|IsNotEmpty _outactors_2)
                          (for _array_element_2 _outactors_2
                            (if (Actor|ActorHasTag _array_element_2 "Henchmen")
                              (Fortnite|FortShippingLog self (Utilities|String|Append "Disabled Henchmen: " (Utilities|GetDisplayName _array_element_2)))
                              (Class|FortAthenaPatrolPath|SetPatrolPathEnabled _array_element_2 false)
                              (Class|FortAthenaLivingWorldPointProviderInterface|DisablePointProvider(Message) _array_element_2)))
                          (else
                            _))
                        (bind _outfortpawns (Fortnite|GetFortPawnsInRange (Transformation|GetActorLocation) 1.874e+04))
                        (if (Utilities|Array|IsNotEmpty _outfortpawns)
                          (for _array_element_3 _outfortpawns
                            (Utilities|Casting|CastToBP_PlayerPawn_NonParticipant _array_element_3
                              (:then
                                (Actor|DestroyActor _array_element_3))
                              (:CastFailed
                                (Utilities|Casting|CastToNPC_Pawn_Irwin_Simple_Sprite _array_element_3)
                                (Actor|DestroyActor _array_element_3))))
                          (else
                            _))
                        ; To enable Superman Sprite providers
                        (bind _outactors_3 (Fortnite|GetAllActorsOfClassinRange "/Script/LagerRuntime.FortAthenaLivingWorldStaticPointProvider" (Transformation|GetActorLocation) 1.874e+04))
                        (if (Utilities|Array|IsNotEmpty _outactors_3)
                          (for _array_element_4 _outactors_3
                            (if (== (Class|FortAthenaLivingWorldPointProviderInterface|GetFiltersTags _array_element_4) (Variables|Default|GetSpriteFilterTag))
                              (Class|FortAthenaLivingWorldPointProviderInterface|EnablePointProvider_DEPRECATED _array_element_4)
                              (Utilities|Array|AddUnique (Variables|Default|GetSpritePointProviders) _array_element_4))))
                        (Utilities|FlowControl|Delay 1.0)
                        ; Destroy pickups in the area
                        (bind _outactors_4 (Fortnite|GetAllActorsOfClassinRange "/Script/FortniteGame.FortPickup" (Transformation|GetActorLocation) 1.874e+04))
                        (if (Utilities|Array|IsNotEmpty _outactors_4)
                          (for _array_element_5 _outactors_4
                            (Actor|DestroyActor _array_element_5))))))
                  (else
                    (Fortnite|FortShippingLog self "Safe Zone Calendar Event is Inactive")))))))))))

(event Custom|PC_DisableRespawn
  (bind _respawnmutator (Variables|Default|GetRespawnMutator))
  ;[0 0.15 0.05105 0.5] If the pre-event calendar event is active, load the event zone data layer.
  (Networking|SwitchHasAuthority
    (:Authority
      (Utilities|Time|ClearandInvalidateTimerbyHandle (Variables|Default|GetEventTimeCheckHandle))
      (Utilities|IsValid _respawnmutator
        (:Is%20Valid
          (Mutator|SetMutatorActive _respawnmutator false)
          (Actor|DestroyActor _respawnmutator))))))

(event EventEndPlay
  (params (EndPlayReason :type EEndPlayReason%20Enum :default "Destroyed"))
  (bind _respawnmutator (Variables|Default|GetRespawnMutator))
  ;[0 0.15 0.05105 0.5] If the pre-event calendar event is active, load the event zone data layer.
  (Networking|SwitchHasAuthority
    (:Authority
      (Utilities|Time|ClearandInvalidateTimerbyHandle (Variables|Default|GetEventTimeCheckHandle))
      (Utilities|IsValid _respawnmutator
        (:Is%20Valid
          (Mutator|SetMutatorActive _respawnmutator false)
          (Actor|DestroyActor _respawnmutator))))))

(event Custom|Activate%20Mutators
  (bind _globalspawntime (Variables|Default|GetGlobalSpawnTime))
  (bind _returnvalue (ScalableFloat|GetValueasIntegerwithContext _globalspawntime))
  (bind _respawnmutator (Variables|Default|GetRespawnMutator))
  ;[0 0.15 0.03756 0.5] Spawn and activate mutators depending on how much time is left before Main Event starts.
  (bind (_eventname _timeuntilend) (MCPContext|WatchCalendarEvent (Variables|Default|GetCalendarEventToWatch))
    (:then
      (if (Variables|Cheat|GetLoadedWithCheats)
        _))
    (:OnEventActive
      _)
    (:OnEventEnded
      (CallFunction|OnEventStart)))
  (Utilities|IsValid _respawnmutator
    (:Is%20Valid
      _)
    (:Is%20Not%20Valid
      (bind _returnvalue_2 (Class|SpecialEventGameplayLibrary|SpawnMutator (select (select (Variables|Cheat|GetLoadedWithCheats) (not (Variables|Cheat|GetLoadedWithCheats)) (<= (.Minutes (Math|Timespan|BreakTimespan _timeuntilend)) _returnvalue)) (Variables|Default|GetGlobalRespawnMutator) (Variables|Default|GetZoneRespawnMutator))))
      (Variables|Default|SetRespawnMutator _returnvalue_2)))
  (Mutator|SetMutatorActive _respawnmutator true)
  (bind _asfort_athena_mutator_se_respawn (Utilities|Casting|CastToFortAthenaMutator_SE_Respawn _respawnmutator))
  (Class|FortAthenaMutatorSERespawn|OverrideRespawnArea _asfort_athena_mutator_se_respawn (Variables|Default|GetRespawn%20Area))
  (CallFunction|WatchforTimeBeforeEvent (<= (.Minutes (Math|Timespan|BreakTimespan _timeuntilend)) _returnvalue))
  (Development|PrintString "Respawn should be active" true true "(R=0.000000,G=0.660000,B=1.000000,A=1.000000)" 3e+01)
  (Utilities|IsValid (Variables|Default|GetEndGameMutator)
    (:Is%20Valid
      _)
    (:Is%20Not%20Valid
      (bind _returnvalue_3 (Class|SpecialEventGameplayLibrary|SpawnMutator (Variables|Default|GetEndGameMutatorClass)))
      (Variables|Default|SetEndGameMutator _returnvalue_3)))
  (Mutator|SetMutatorActive (Variables|Default|GetEndGameMutator) true)
  (bind (_asynctaskproxy _payload) (GameplayEventRouter|StartListeningToStatefulEvent (GameplayEventRouter|GetGlobalEventRouter) "/Events_Shared/Gameplay/GlobalEvents/Event_RouterEvent_String.Event_RouterEvent_String")
    (:OnEventReceived
      (bind _payload_eventname_4_40c360d849aa0a7e7043d1be14586049 (Utilities|Struct|BreakEventRouterEventString _payload))
      (if (Utilities|String|Equal,CaseInsensitive(String) _payload_eventname_4_40c360d849aa0a7e7043d1be14586049 "InventoryModDisable")
        (Variables|Default|SetInventoryPhase2 (Utilities|String|Equal,CaseInsensitive(String) _payload_eventname_4_40c360d849aa0a7e7043d1be14586049 "InventoryModDisable"))
        (elif (Utilities|String|Equal,CaseInsensitive(String) _payload_eventname_4_40c360d849aa0a7e7043d1be14586049 "PC_EventEnded")
          (CallFunction|PostEvent))))))

(event Custom|PC_SwapRespawnMutator
  (bind _respawnmutator (Variables|Default|GetRespawnMutator))
  (bind _respawnmutator_2 (Variables|Default|GetRespawnMutator))
  ;[0 0.15 0.03173 0.5] Swap Respawn mode to Global.
  (Networking|SwitchHasAuthority
    (:Authority
      (Utilities|IsValid _respawnmutator
        (:Is%20Valid
          (Mutator|SetMutatorActive _respawnmutator false)
          (Actor|DestroyActor _respawnmutator)
          (Development|PrintString "Destroy old mutator" true true "(R=0.000000,G=0.660000,B=1.000000,A=1.000000)" 5e+01)
          (Utilities|FlowControl|Delay 0.5))
        (:Is%20Not%20Valid
          _))
      (bind _returnvalue (Class|SpecialEventGameplayLibrary|SpawnMutator (Variables|Default|GetGlobalRespawnMutator)))
      (Variables|Default|SetRespawnMutator _returnvalue)
      (Utilities|IsValid _respawnmutator_2
        (:Is%20Valid
          (Mutator|SetMutatorActive _respawnmutator_2 true)
          (bind _asfort_athena_mutator_se_respawn (Utilities|Casting|CastToFortAthenaMutator_SE_Respawn _respawnmutator_2))
          (Class|FortAthenaMutatorSERespawn|OverrideRespawnArea _asfort_athena_mutator_se_respawn (Variables|Default|GetRespawn%20Area))
          (Development|PrintString "Mutator Swapped" true true "(R=0.000000,G=0.660000,B=1.000000,A=1.000000)" 2e+01))
        (:Is%20Not%20Valid
          (Development|PrintString "Respawn Mutator is INVALID!" true true "(R=0.000000,G=0.660000,B=1.000000,A=1.000000)" 3e+01))))))

(event Custom|StartGlobalRespawn
  (bind _respawnmutator (Variables|Default|GetRespawnMutator))
  (bind _respawnmutator_2 (Variables|Default|GetRespawnMutator))
  ;[0 0.15 0.03173 0.5] Swap Respawn mode to Global.
  (Networking|SwitchHasAuthority
    (:Authority
      (Utilities|IsValid _respawnmutator
        (:Is%20Valid
          (Mutator|SetMutatorActive _respawnmutator false)
          (Actor|DestroyActor _respawnmutator)
          (Development|PrintString "Destroy old mutator" true true "(R=0.000000,G=0.660000,B=1.000000,A=1.000000)" 5e+01)
          (Utilities|FlowControl|Delay 0.5))
        (:Is%20Not%20Valid
          _))
      (bind _returnvalue (Class|SpecialEventGameplayLibrary|SpawnMutator (Variables|Default|GetGlobalRespawnMutator)))
      (Variables|Default|SetRespawnMutator _returnvalue)
      (Utilities|IsValid _respawnmutator_2
        (:Is%20Valid
          (Mutator|SetMutatorActive _respawnmutator_2 true)
          (bind _asfort_athena_mutator_se_respawn (Utilities|Casting|CastToFortAthenaMutator_SE_Respawn _respawnmutator_2))
          (Class|FortAthenaMutatorSERespawn|OverrideRespawnArea _asfort_athena_mutator_se_respawn (Variables|Default|GetRespawn%20Area))
          (Development|PrintString "Mutator Swapped" true true "(R=0.000000,G=0.660000,B=1.000000,A=1.000000)" 2e+01))
        (:Is%20Not%20Valid
          (Development|PrintString "Respawn Mutator is INVALID!" true true "(R=0.000000,G=0.660000,B=1.000000,A=1.000000)" 3e+01))))))

(event Custom|CheckForTimeLeft
  ;[0 0.15 0.09158 0.5] If event is active, check if we are close enough to the event time to swap to global respawn.
  (Networking|SwitchHasAuthority
    (:Authority
      (bind (_biseventactive _timesincebegin _timeuntilend) (CalendarEvents|GetCalendarEventInformation (Variables|Default|GetCalendarEventToWatch)))
      (if _biseventactive
        (if (<= (.Minutes (Math|Timespan|BreakTimespan _timeuntilend)) (ScalableFloat|GetValueasIntegerwithContext (Variables|Default|GetGlobalSpawnTime)))
          (Variables|Default|SetInventoryPhase2 true)
          (Utilities|Time|ClearandInvalidateTimerbyHandle (Variables|Default|GetEventTimeCheckHandle))
          (CallFunction|StartGlobalRespawn)
          (CallFunction|AddWeaponSpawns))
        (else
          (Utilities|Time|ClearandInvalidateTimerbyHandle (Variables|Default|GetEventTimeCheckHandle)))))))

(event Custom|PC_Zone2
  ;[0 0.15 0.09158 0.5] If event is active, check if we are close enough to the event time to swap to global respawn.
  (Development|ExecuteConsoleCommand "Cheat ke* DebugCountdown 900") ; Seconds in 15 minutes for timer.
  (Variables|Default|SetInventoryPhase2 true)
  (Utilities|Time|ClearandInvalidateTimerbyHandle (Variables|Default|GetEventTimeCheckHandle))
  (CallFunction|StartGlobalRespawn)
  (CallFunction|AddWeaponSpawns))

(event Custom|WatchForTimeBeforeEvent
  (params (InPointOfNoReturnTime :type Boolean))
  ;[0 0.15 0.09158 0.5] If event is active, check if we are close enough to the event time to swap to global respawn.
  (if (not InPointOfNoReturnTime)
    (bind _returnvalue (Utilities|Time|SetTimerbyEvent (EventDispatchers|CreateEvent :!event Custom|CheckForTimeLeft) 0.1 true))
    (Variables|Default|SetEventTimeCheckHandle _returnvalue)))

(event Custom|AddWeaponSpawns
  ;[0 0.15 0.09158 0.5] Add weapon spawns to training dummy
  (bind _outactors (Actor|GetAllActorsOfClass "/PistolCanary_Training/Gameplay/TrainingDummy/BGA_TrainingDummy_Target.BGA_TrainingDummy_Target_C"))
  (for _array_element _outactors
    (Variables|/PistolCanary_Training/Gameplay/TrainingDummy/BGA_TrainingDummy_Target.BGA_TrainingDummy_Target_C|SetbIsPreEventEnding true _array_element)))

(event Custom|ClearTrainingDummies
  ;[0 0.15 0.09158 0.5] Add weapon spawns to training dummy
  (bind _outactors (Actor|GetAllActorsOfClass "/PistolCanary_Training/Gameplay/TrainingDummy/BGA_TrainingDummy_Target.BGA_TrainingDummy_Target_C"))
  (for _array_element _outactors
    (Variables|/PistolCanary_Training/Gameplay/TrainingDummy/BGA_TrainingDummy_Target.BGA_TrainingDummy_Target_C|SetCanRespawn false _array_element)
    (Class|BGATrainingDummyTarget|ForceSelfDestruct _array_element)
    (Actor|SetLifeSpan _array_element 1.0)))

(event Custom|CleanUpSprites
  (bind _spritepointproviders (Variables|Default|GetSpritePointProviders))
  (if (Utilities|Array|IsNotEmpty _spritepointproviders)
    (for _array_element _spritepointproviders
      (if (== (Utilities|Array|Get(aref) (Variables|/Script/Engine.Actor|GetTags _array_element)) (Variables|Default|GetTagName_DisableSpriteProvider))
        (Actor|DestroyActor _array_element)
        (else
          (Class|FortAthenaLivingWorldPointProviderInterface|DisablePointProvider _array_element))))
    (bind _outfortpawns (Fortnite|GetFortPawnsInRange (Transformation|GetActorLocation) 1.874e+04))
    (if (Utilities|Array|IsNotEmpty _outfortpawns)
      (for _array_element_2 _outfortpawns
        (Utilities|Casting|CastToNPC_Pawn_Irwin_Simple_Sprite _array_element_2)
        (Actor|DestroyActor _array_element_2)))))

(event Custom|OnEventStart
  ;[0 0.15 0.09158 0.5] Event Started
  (CallFunction|ClearTrainingDummies)
  (CallFunction|CleanUpSprites)
  (CallFunction|ClearImmuneGE)
  (Development|PrintString "Event Starting")
  (Class|FortMemoryBPUtils|ForceGarbageCollection false))

(event Custom|PC_Event
  ;[0 0.15 0.09158 0.5] Event Started
  (CallFunction|ClearTrainingDummies)
  (CallFunction|CleanUpSprites)
  (CallFunction|ClearImmuneGE)
  (Development|PrintString "Event Starting")
  (Class|FortMemoryBPUtils|ForceGarbageCollection false))

(event Custom|PC_ZoneRules
  (bind _outfortplayerpawns (Fortnite|GetFortPlayerPawns))
  (for _array_element _outfortplayerpawns
    (GameplayEffects|RemoveActiveGameplayEffectbySourceEffect (Ability|GetFortAbilitySystemComponent _array_element) "/PistolCanary_Training/Gameplay/Abilities/GE_PistolCanary_EventDamageImmune_ZoneRules.GE_PistolCanary_EventDamageImmune_ZoneRules_C" 0 -1)
    (GameplayEffects|RemoveActiveGameplayEffectbySourceEffect (Ability|GetFortAbilitySystemComponent _array_element) "/PistolCanary_Training/Gameplay/Abilities/GE_PistolCanary_EventDamageImmune.GE_PistolCanary_EventDamageImmune_C" 0 -1)
    (if (Pawn|IsPlayerControlled _array_element)
      (GameplayEffects|ApplyGameplayEffectToTarget :self (Ability|GetFortAbilitySystemComponent _array_element) :GameplayEffectClass "/PistolCanary_Training/Gameplay/Abilities/GE_PistolCanary_EventDamageImmune_EventRules.GE_PistolCanary_EventDamageImmune_EventRules_C" :Target (Ability|GetFortAbilitySystemComponent _array_element) :Level 0.0)
      (GameplayEffects|ApplyGameplayEffectToTarget :self (Ability|GetFortAbilitySystemComponent _array_element) :GameplayEffectClass "/PistolCanary_Training/Gameplay/Effects/GE_PistolCanary_EventNoAmmoCost.GE_PistolCanary_EventNoAmmoCost_C" :Target (Ability|GetFortAbilitySystemComponent _array_element) :Level 0.0)))
  (CallFunction|UpdateRespawnAreaforEvent))

(event Custom|ClearImmuneGE
  (bind _outfortplayerpawns (Fortnite|GetFortPlayerPawns))
  (for _array_element _outfortplayerpawns
    (GameplayEffects|RemoveActiveGameplayEffectbySourceEffect (Ability|GetFortAbilitySystemComponent _array_element) "/PistolCanary_Training/Gameplay/Abilities/GE_PistolCanary_EventDamageImmune_ZoneRules.GE_PistolCanary_EventDamageImmune_ZoneRules_C" 0 -1)
    (GameplayEffects|RemoveActiveGameplayEffectbySourceEffect (Ability|GetFortAbilitySystemComponent _array_element) "/PistolCanary_Training/Gameplay/Abilities/GE_PistolCanary_EventDamageImmune.GE_PistolCanary_EventDamageImmune_C" 0 -1)
    (if (Pawn|IsPlayerControlled _array_element)
      (GameplayEffects|ApplyGameplayEffectToTarget :self (Ability|GetFortAbilitySystemComponent _array_element) :GameplayEffectClass "/PistolCanary_Training/Gameplay/Abilities/GE_PistolCanary_EventDamageImmune_EventRules.GE_PistolCanary_EventDamageImmune_EventRules_C" :Target (Ability|GetFortAbilitySystemComponent _array_element) :Level 0.0)
      (GameplayEffects|ApplyGameplayEffectToTarget :self (Ability|GetFortAbilitySystemComponent _array_element) :GameplayEffectClass "/PistolCanary_Training/Gameplay/Effects/GE_PistolCanary_EventNoAmmoCost.GE_PistolCanary_EventNoAmmoCost_C" :Target (Ability|GetFortAbilitySystemComponent _array_element) :Level 0.0)))
  (CallFunction|UpdateRespawnAreaforEvent))

(event Custom|pistolcanary
  ;[0 0.15 0.09158 0.5] Event Started
  (CallFunction|ClearTrainingDummies)
  (CallFunction|CleanUpSprites)
  (CallFunction|ClearImmuneGE)
  (Development|PrintString "Event Starting")
  (Class|FortMemoryBPUtils|ForceGarbageCollection false))

(event Custom|PC_ItemSpawn
  (GameplayEventRouter|BroadcastStatefulEvent (GameplayEventRouter|GetGlobalEventRouter) (Utilities|Struct|MakeEventRouterEventString (Variables|Default|GetAmmoEventName)) 0))

(event Custom|BroadcastVerb
  (bind _outfortplayerpawns (Fortnite|GetFortPlayerPawns))
  (for _array_element _outfortplayerpawns
    (CustomVerb|BroadcastCustomVerbMessage (CustomVerb|MakeCustomVerbMessage _array_element "(TagName=\"CustomVerb.PistolCanary.HideUrgentQuest\")"))))

(event Custom|PostEvent
  (bind _outfortplayerpawns (Fortnite|GetFortPlayerPawns))
  (for _array_element _outfortplayerpawns
    (Utilities|IsValid (Ability|GetFortAbilitySystemComponent _array_element)
      (:Is%20Valid
        (GameplayEffects|RemoveActiveGameplayEffectbySourceEffect (Ability|GetFortAbilitySystemComponent _array_element) "/PistolCanary_Training/Gameplay/Effects/GE_PistolCanary_EventNoAmmoCost.GE_PistolCanary_EventNoAmmoCost_C" 0 -1))))
  (Mutator|SetMutatorActive (Variables|Default|GetRespawnMutator) false)
  (Mutator|SetMutatorActive (Variables|Default|GetEndGameMutator) false)
  (bind _returnvalue (GameplayEventRouter|BroadcastEvent (GameplayEventRouter|GetGlobalEventRouter) (Utilities|Struct|MakeMMHenchmenManagerEvents "NewEnumerator0") 0))
  (Fortnite|FortShippingLog self (Utilities|String|Append "EVENT ZONE MANAGER:: MMHenchman ENABLE PATHS EVENT:: ANY LISTENERS? " (Utilities|String|ToString(Boolean) _returnvalue))))

(event Custom|UpdateRespawnAreaForEvent
  (bind _respawnmutator (Variables|Default|GetRespawnMutator))
  (Networking|SwitchHasAuthority
    (:Authority
      (Utilities|IsValid (Variables|Default|GetRespawnMutator)
        (:Is%20Valid
          _)
        (:Is%20Not%20Valid
          (bind _returnvalue (Class|SpecialEventGameplayLibrary|SpawnMutator (Variables|Default|GetGlobalRespawnMutator)))
          (Variables|Default|SetRespawnMutator _returnvalue)))
      (Mutator|SetMutatorActive _respawnmutator true)
      (bind _asfort_athena_mutator_se_respawn (Utilities|Casting|CastToFortAthenaMutator_SE_Respawn _respawnmutator))
      (Class|FortAthenaMutatorSERespawn|OverrideRespawnArea _asfort_athena_mutator_se_respawn (Variables|Default|GetEventRespawnArea)))))

(fn ConstructionScript)
