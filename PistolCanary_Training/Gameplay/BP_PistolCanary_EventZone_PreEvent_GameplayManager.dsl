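; Gameplay manager for the PistolCanary pre-event zone that handles training dummies and leaderboard-based player buffs.
; A looping timer has the dedicated server check tracked players' mesh leaderboard ranks, granting tiered effects to the local top three and a Giga buff to the global top 25.
; An incremented RepNotify counter flags event start, ending the timer.
; Dummy respawns set a pre-event-ending flag from calendar time remaining or the five-minutes-remaining flag.
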(event EventBeginPlay
  (|CallParentFunction)
  (bind _returnvalue (Utilities|Time|SetTimerbyEvent (EventDispatchers|CreateEvent :!event Custom|CheckLeaderBoard) 3.0 true))
  (Variables|Default|SetTH_LeaderBoard _returnvalue))

(event Custom|5MinutesRemain
  (Variables|Default|SetSpawnExtraWeapons true))

(event Custom|RespawnDummy
  (params (TrainingDummy :type BGA%20Training%20Dummy%20Target%20Class%20Reference) (ActorTransfoorm :type Transform))
  (bind (_biseventactive _timesincebegin _timeuntilend) (CalendarEvents|GetCalendarEventInformation (Variables|Default|GetCalendarEventToWatch)))
  (if _biseventactive
    (bind _returnvalue (Game|SpawnActorfromClass TrainingDummy ActorTransfoorm "Undefined" "MultiplyWithRoot" 0 0 0))
    (Variables|/PistolCanary_Training/Gameplay/TrainingDummy/BGA_TrainingDummy_Target.BGA_TrainingDummy_Target_C|SetbIsPreEventEnding (select (<= (.Minutes (Math|Timespan|BreakTimespan _timeuntilend)) (ScalableFloat|GetValueasIntegerwithContext (Variables|Default|GetGlobalSpawnTime))) true false) _returnvalue)
    (else
      (bind _returnvalue_2 (Game|SpawnActorfromClass TrainingDummy ActorTransfoorm "Undefined" "MultiplyWithRoot" 0 0 0))
      (Variables|/PistolCanary_Training/Gameplay/TrainingDummy/BGA_TrainingDummy_Target.BGA_TrainingDummy_Target_C|SetbIsPreEventEnding (select (Variables|Default|GetSpawnExtraWeapons) true false) _returnvalue_2))))

(event Custom|AddPlayerToLeaderBoard
  (params (PlayerPawn :type Fort%20Player%20Pawn%20Athena%20Object%20Reference))
  ; PROTO
  (Utilities|Array|AddUnique (Variables|Default|GetListOfActivePlayers) PlayerPawn))

(event Custom|ClearTrainingDummies
  (params (TrainingDummies :type BGA%20Training%20Dummy%20Target%20Object%20Reference))
  (Variables|/PistolCanary_Training/Gameplay/TrainingDummy/BGA_TrainingDummy_Target.BGA_TrainingDummy_Target_C|SetCanRespawn false TrainingDummies)
  (Class|BGATrainingDummyTarget|ForceSelfDestruct TrainingDummies)
  (Actor|SetLifeSpan TrainingDummies 1.0))

(event Custom|RemovePlayerFromLeaderBoard
  (params (PlayerPawn :type Fort%20Player%20Pawn%20Athena%20Object%20Reference%20%28by%20ref%29))
  ; PROTO
  (Utilities|Array|RemoveItem (Variables|Default|GetListOfActivePlayers) PlayerPawn))

(event Custom|CheckLeaderBoard
  (bind _returnvalue (Class|SpecialEventMeshLeaderboardManager|GetMeshLeaderboardManager false))
  (bind _listofactiveplayers (Variables|Default|GetListOfActivePlayers))
  ; PROTO
  (if (Variables|Default|GetEventStarted)
    (Utilities|Time|ClearandInvalidateTimerbyHandle (Variables|Default|GetTH_LeaderBoard))
    (elif (Networking|IsDedicatedServer)
      (if (Utilities|Array|IsNotEmpty _listofactiveplayers)
        (for _array_element _listofactiveplayers
          (if (== (Class|SpecialEventMeshLeaderboardManager|GetPlayerPositioninGlobalLeaderboard _returnvalue (Variables|/Script/Engine.Pawn|GetPlayerState _array_element)) -1)
            (bind _returnvalue_2 (Class|SpecialEventMeshLeaderboardManager|GetScoreforPlayer _returnvalue (Variables|/Script/Engine.Pawn|GetPlayerState _array_element)))
            (if (>= _returnvalue_2 1.0)
              (switch int (Class|SpecialEventMeshLeaderboardManager|GetPlayerPositioninLocalLeaderboard _returnvalue (Variables|/Script/Engine.Pawn|GetPlayerState _array_element))
                (:1
                  (if (not (GameplayTags|HasMatchingGameplayTag (Ability|GetAbilitySystemComponent _array_element) "(TagName=\"PistolCanary.PreEvent.PvP.Buff.1st\")"))
                    (Development|PrintString "1stBuff")
                    (GameplayEffects|RemoveActiveEffectswithGrantedTags (Ability|GetAbilitySystemComponent _array_element) "(GameplayTags=((TagName=\"PistolCanary.PreEvent.PvP.Buff.2nd\"),(TagName=\"PistolCanary.PreEvent.PvP.Buff.3rd\"),(TagName=\"PistolCanary.PreEvent.PvP.Buff.Giga\")))")))
                (:2
                  (if (not (GameplayTags|HasMatchingGameplayTag (Ability|GetAbilitySystemComponent _array_element) "(TagName=\"PistolCanary.PreEvent.PvP.Buff.2nd\")"))
                    (Development|PrintString "2ndBuff")
                    (GameplayEffects|RemoveActiveEffectswithGrantedTags (Ability|GetAbilitySystemComponent _array_element) "(GameplayTags=((TagName=\"PistolCanary.PreEvent.PvP.Buff.3rd\"),(TagName=\"PistolCanary.PreEvent.PvP.Buff.Giga\"),(TagName=\"PistolCanary.PreEvent.PvP.Buff.1st\")))")))
                (:3
                  (if (not (GameplayTags|HasMatchingGameplayTag (Ability|GetAbilitySystemComponent _array_element) "(TagName=\"PistolCanary.PreEvent.PvP.Buff.3rd\")"))
                    (Development|PrintString "3rdBuff")
                    (GameplayEffects|RemoveActiveEffectswithGrantedTags (Ability|GetAbilitySystemComponent _array_element) "(GameplayTags=((TagName=\"PistolCanary.PreEvent.PvP.Buff.2nd\"),(TagName=\"PistolCanary.PreEvent.PvP.Buff.Giga\"),(TagName=\"PistolCanary.PreEvent.PvP.Buff.1st\")))"))))
              (GameplayEffects|ApplyGameplayEffectToSelf :self (Ability|GetAbilitySystemComponent _array_element) :GameplayEffectClass (Utilities|Select :Index (Class|SpecialEventMeshLeaderboardManager|GetPlayerPositioninLocalLeaderboard _returnvalue (Variables|/Script/Engine.Pawn|GetPlayerState _array_element)) :Option%201 "/PistolCanary_Training/Gameplay/Abilities/GE_PistolCanary_1stBuff.GE_PistolCanary_1stBuff_C" :Option%202 "/PistolCanary_Training/Gameplay/Abilities/GE_PistolCanary_2ndBuff.GE_PistolCanary_2ndBuff_C" :Option%203 "/PistolCanary_Training/Gameplay/Abilities/GE_PistolCanary_3rdBuff.GE_PistolCanary_3rdBuff_C") :Level 0.0))))
        (for _array_element_2 _listofactiveplayers
          (if (== (Class|SpecialEventMeshLeaderboardManager|GetPlayerPositioninGlobalLeaderboard (Class|SpecialEventMeshLeaderboardManager|GetMeshLeaderboardManager false) (Variables|/Script/Engine.Pawn|GetPlayerState _array_element_2)) -1)
            (if (>= (Class|SpecialEventMeshLeaderboardManager|GetPlayerPositioninLocalLeaderboard (Class|SpecialEventMeshLeaderboardManager|GetMeshLeaderboardManager false) (Variables|/Script/Engine.Pawn|GetPlayerState _array_element_2)) 4)
              (GameplayEffects|RemoveActiveEffectswithSourceTags (Ability|GetAbilitySystemComponent _array_element_2) "(GameplayTags=((TagName=\"Gameplay.Mod.FireRate\")))"))))
        (for _array_element_3 _listofactiveplayers
          (if (and (>= (Class|SpecialEventMeshLeaderboardManager|GetPlayerPositioninGlobalLeaderboard (Class|SpecialEventMeshLeaderboardManager|GetMeshLeaderboardManager false) (Variables|/Script/Engine.Pawn|GetPlayerState _array_element_3)) 0) (<= (Class|SpecialEventMeshLeaderboardManager|GetPlayerPositioninGlobalLeaderboard (Class|SpecialEventMeshLeaderboardManager|GetMeshLeaderboardManager false) (Variables|/Script/Engine.Pawn|GetPlayerState _array_element_3)) 24))
            (if (not (GameplayTags|HasMatchingGameplayTag (Ability|GetAbilitySystemComponent _array_element_3) "(TagName=\"PistolCanary.PreEvent.PvP.Buff.Giga\")"))
              (Development|PrintString "GigaBuff")
              (GameplayEffects|RemoveActiveEffectswithGrantedTags (Ability|GetAbilitySystemComponent _array_element_3) "(GameplayTags=((TagName=\"PistolCanary.PreEvent.PvP.Buff.2nd\"),(TagName=\"PistolCanary.PreEvent.PvP.Buff.1st\"),(TagName=\"PistolCanary.PreEvent.PvP.Buff.3rd\")))")
              (GameplayEffects|ApplyGameplayEffectToSelf :self (Ability|GetAbilitySystemComponent _array_element_3) :GameplayEffectClass "/PistolCanary_Training/Gameplay/Abilities/GE_PistolCanary_GigaBuff.GE_PistolCanary_GigaBuff_C" :Level 0.0))))))))

(event Custom|SetEventStarted
  ; PROTO
  (Variables|Default|SetEventStartedRep (+ (Variables|Default|GetEventStartedRep) 1)))

(fn CheckGlobalBuffTag
  (params (PlayerTag :type Gameplay%20Tag%20Asset%20Interface%20Interface))
  (return (GameplayTags|HasMatchingGameplayTag PlayerTag "(TagName=\"PistolCanary.PreEvent.PvP.Buff.Giga\")")))

(fn OnRep_EventStartedRep
  (Variables|Default|SetEventStarted true))

(fn ConstructionScript)
